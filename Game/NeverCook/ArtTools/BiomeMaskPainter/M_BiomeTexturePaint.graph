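# Translucent unlit material in the biome mask painter art tools, apparently for displaying the BiomeMask texture while it is painted.
# It samples BiomeMask, marked as the default mesh-paint texture, as linear color at unshifted UVs and feeds its RGB into a Desaturation node.
# A UV offset sum from UVOffset_X and UVOffset_Y does not feed the sample.
# An Opacity parameter is present, but no connections to material outputs are shown.
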Material /Game/NeverCook/ArtTools/BiomeMaskPainter/M_BiomeTexturePaint
graph {
 properties {bAutomaticallySetUsageInEditor=false, blendMode=BLEND_Translucent, shadingModel=MSM_Unlit, usedShadingModels=Unlit}
 nodes [
  n1 = TextureCoordinate()
  Add(A=@n1, B=AppendVector(A=ScalarParameter() {parameterName=UVOffset_X}, B=ScalarParameter() {parameterName=UVOffset_Y}))
  n2 = TextureSampleParameter2D(UVs=@n1) {
   isDefaultMeshpaintTexture=true
   parameterName=BiomeMask
   samplerType=SAMPLERTYPE_LinearColor
   texture=&/T_BiomeMask_Template
  }
  Desaturation(None=@n2.RGB)
  ScalarParameter() {defaultValue=0.5, parameterName=Opacity}
 ]
}
