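# Editor utility actor for painting biome masks, whose construction script fits four paint planes over the level's landscape.
# It derives component size from landscape proxy bounds, centres itself above the landscape and scales by tile count when Enabled.
# Each plane gets a dynamic paint material fed the BiomeTexture mask, and the Paint Region enum shows only the selected quadrant's plane.
# It also builds a Usage legend mapping each quadrant's RGB channels to biome names.
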EditorUtilityBlueprint /Game/NeverCook/ArtTools/BiomeMaskPainter/EUBP_BiomeMaskPainter
graphs [
 EventGraph {nodes [
   Event BeginPlay()
   Event ActorBeginOverlap()
   Event Tick!() -> (OutputDelegate, then, DeltaSeconds:float)
  ]}
 UserConstructionScript {nodes [
   n1 = GetAllActorsOfClass!(execute, WorldContextObject:Object Reference, ActorClass:Actor Class Reference=/Script/Landscape.LandscapeProxy) -> (then, OutActors:Array of Landscape Proxy Object References)
   > Branch(Condition="integer > integer"(A=Length(TargetArray:Array of Landscape Proxy Object References=@n1)))
   > n2 = For Each Loop(Array:Array of Landscape Proxy Object References=@n1) -> (Array Element:Landscape Proxy Object Reference)
   Clear(execute=@n2.Completed, TargetArray:Array of Landscape Proxy Object References=@n1)
   Is Valid(exec=@n2.LoopBody, InputObject=@n2)
   n3 = Get Actor Bounds(self=@n2, bOnlyCollidingComponents=false)
   n4 = BreakVector(InVec=@n3.BoxExtent)
   n5 = Get LandscapeComponentSize() -> (LandscapeComponentSize:double)
   n6 = SelectFloat(A=@n4, B=@n5, bPickA="float > float"(A=@n4, B=@n5))
   > Set LandscapeComponentSize(LandscapeComponentSize:double="To Float (Integer)"(InInt=Round(A=@n6) {is=title K2Node_CallFunction})) -> (Output_Get:double)
   n7 = GetOwner(self:Actor Object Reference=@n2)
   n8 = "To String (Vector)"!(InVec:Vector=@n3.BoxExtent) -> (ReturnValue:String)
   > PrintString(InString=Append(A=Append(A="Landscape Component size =", B=@n8), B=GetDisplayName(Object=@n7)))
   Construction Script()
   > n9 = Sequence!(execute) -> (then_0, then_1, then_2:Exec, then_3:Exec)
   Add(
    execute=@n9.then_1
    TargetArray:Array of Static Mesh Component Object References=Get Meshes() -> (Meshes:Array of Static Mesh Component Object References)
    NewItem:"Static Mesh Component Object Reference (by ref)"=Get BiomePaintPlane() -> (BiomePaintPlane:Static Mesh Component Object Reference)
   ) {is=title K2Node_CallArrayFunction}
   > Add(TargetArray=Get Meshes(), NewItem=Get BiomePaintPlane2() -> (BiomePaintPlane2:Static Mesh Component Object Reference))
   > Add(TargetArray=Get Meshes(), NewItem=Get BiomePaintPlane3() -> (BiomePaintPlane3:Static Mesh Component Object Reference))
   > Add(TargetArray=Get Meshes(), NewItem=Get BiomePaintPlane4() -> (BiomePaintPlane4:Static Mesh Component Object Reference))
   n10 = Sequence!(execute=@n9.then_3) -> (then_0, then_1, then_2, then_3, then_4:Exec)
   > n11 = Get!(execute:Exec) -> (then:Exec, else:Exec, Lanscape Actor:Landscape Object Reference)
   n12 = Get Lanscape Actor() -> (Lanscape Actor:Landscape Object Reference)
   n13 = BreakVector(InVec=Get Actor Location(self=@n12))
   n14 = float / float(
    A=float * float(A=Get LandscapeComponentSize(), B=Get Number of Landscape Tiles() -> (Number of Landscape Tiles:double)) {is=title K2Node_PromotableOperator}
    B=2
   )
   Set Actor Location!(
    execute=[@n19.then, @n11.then]
    self:Actor Object Reference
    NewLocation:Vector=MakeVector(X=float + float(A=@n13, B=@n14), Y=float + float(A=@n13.Y, B=@n14), Z=10000)
    bSweep:bool=false
    bTeleport:bool=false
   ) -> (then, SweepHitResult:Hit Result Structure, ReturnValue:bool) {is=title K2Node_CallFunction}
   n15 = float / float(A=SelectFloat(A=@n14, B=1, bPickA=Get Enabled() -> (Enabled:bool)), B=100)
   > SetActorScale3D!(execute, self:Actor Object Reference, NewScale3D:Vector=MakeVector(X=@n15, Y=@n15, Z=@n15)) -> (then)
   n16 = GetAllActorsOfClass(execute=@n11.else, ActorClass=/Script/Landscape) -> (OutActors:Array of Landscape Object References)
   > Branch(Condition="integer > integer"(A=Length(TargetArray:Array of Landscape Object References=@n16)))
   > n17 = Sequence!(execute) -> (then_0, then_1)
   n18 = "Get (a copy)"(Array:Array of Landscape Object References=@n16) -> (Output:Landscape Object Reference)
   > Is Valid(InputObject=@n18)
   > n19 = Set Lanscape Actor(Lanscape Actor:Landscape Object Reference=@n18) -> (Output_Get:Landscape Object Reference)
   Clear(execute=@n17.then_1, TargetArray:Array of Landscape Object References=@n16)
   n20 = GetCurrentLevelName!(execute=@n10.then_2, WorldContextObject:Object Reference, bRemovePrefixString:bool=true) -> (then, ReturnValue:String)
   > PrintString(InString=@n20)
   n21 = Get Quadrant Bottom Right() -> (Quadrant Bottom Right:Array of Strings)
   n22 = "Get (a copy)"(Array:Array of Strings=@n21) -> (Output:String)
   n23 = "Get (a copy)"(Array=@n21, Dimension 1=1)
   n24 = Append(
    A=Append(A=Append(A="\r\nQuadrant Bottom Right \r\n\r\nRed = ", B=@n22), B=Append(A="\r\nGreen = ", B=@n23))
    B=Append(A="\r\nBlue = ", B="Get (a copy)"(Array=@n21, Dimension 1=2))
   )
   n25 = Get Quadrant Bottom Left() -> (Quadrant Bottom Left:Array of Strings)
   n26 = Append(
    A=Append(A="\r\nQuadrant Bottom Left \r\n\r\nRed = ", B="Get (a copy)"(Array=@n25))
    B=Append(A="\r\nGreen = ", B="Get (a copy)"(Array=@n25, Dimension 1=1))
   )
   n27 = Append(A="\r\nBlue = ", B="Get (a copy)"(Array=@n25, Dimension 1=2))
   n28 = Get Quadrant Top Right() -> (Quadrant Top Right:Array of Strings)
   n29 = Append(A="\r\nQuadrant Top Right\r\n\r\nRed = ", B="Get (a copy)"(Array=@n28))
   n30 = Append(A="\r\nGreen = ", B="Get (a copy)"(Array=@n28, Dimension 1=1))
   n31 = "Get (a copy)"(Array=@n28, Dimension 1=2)
   n32 = Get Quadrant Top Left() -> (Quadrant Top Left:Array of Strings)
   n33 = Append(A="\r\nBlue = ", B="Get (a copy)"(Array=@n32, Dimension 1=2))
   n34 = Append(
    A=Append(A="Quadrant Top Left \r\n\r\nRed = ", B="Get (a copy)"(Array=@n32))
    B=Append(A="\r\nGreen = ", B="Get (a copy)"(Array=@n32, Dimension 1=1))
   )
   n35 = Append(A=Append(A=Append(A=@n34, B=@n33), B="\r\n"), B=Append(A=Append(A=@n29, B=@n30), B=Append(A="\r\nBlue = ", B=@n31)))
   n36 = Append(A=Append(A=Append(A=@n35, B="\r\n"), B=Append(A=@n26, B=@n27)), B="\r\n")
   > Set Usage(Usage:Text="To Text (String)"(InString=Append(A=@n36, B=@n24))) -> (Output_Get:Text)
   Get!(execute=@n10.then_3) -> (then, else, BiomeTexture:Texture Object Reference)
   > n37 = For Each Loop(Array:Array of Static Mesh Component Object References=Get Meshes()) -> (Array Element:Static Mesh Component Object Reference)
   > Is Valid(InputObject=@n37)
   > n38 = CreateDynamicMaterialInstance!(
    execute
    self:Primitive Component Object Reference=@n37
    ElementIndex:int=0
    SourceMaterial:Material Interface Object Reference=&/M_BiomeTexturePaint
    OptionalName:Name=None
   ) -> (then, ReturnValue:Material Instance Dynamic Object Reference) {is=title K2Node_CallFunction}
   > SetTextureParameterValue(self=@n38, ParameterName=BiomeMask, Value=Get BiomeTexture() -> (BiomeTexture:Texture Object Reference))
   n39 = Switch on BiomeMaskPaintingRegion!(execute=@n10.then_4, Selection:BiomeMaskPaintingRegion Enum=Get Paint Region() -> (Paint Region:BiomeMaskPaintingRegion Enum=NewEnumerator0)) -> (NewEnumerator0:Exec, NewEnumerator1:Exec, NewEnumerator2:Exec, NewEnumerator3:Exec) {is=title K2Node_SwitchEnum}
   > SetVisibility!(execute, self:Scene Component Object Reference=Get BiomePaintPlane(), bNewVisibility:bool=true, bPropagateToChildren:bool=false) -> (then)
   > SetVisibility(self=[Get BiomePaintPlane2(), Get BiomePaintPlane3(), Get BiomePaintPlane4()], bNewVisibility=false, bPropagateToChildren=false)
   SetVisibility(execute=@n39.NewEnumerator1, self=Get BiomePaintPlane2(), bNewVisibility=true, bPropagateToChildren=false)
   > SetVisibility(self=[Get BiomePaintPlane(), Get BiomePaintPlane3(), Get BiomePaintPlane4()], bNewVisibility=false, bPropagateToChildren=false)
   SetVisibility(execute=@n39.NewEnumerator2, self=Get BiomePaintPlane3(), bNewVisibility=true, bPropagateToChildren=false)
   > SetVisibility(self=[Get BiomePaintPlane(), Get BiomePaintPlane2(), Get BiomePaintPlane4()], bNewVisibility=false, bPropagateToChildren=false)
   SetVisibility(execute=@n39.NewEnumerator3, self=Get BiomePaintPlane4(), bNewVisibility=true, bPropagateToChildren=false)
   > SetVisibility(self=[Get BiomePaintPlane(), Get BiomePaintPlane2(), Get BiomePaintPlane3()], bNewVisibility=false, bPropagateToChildren=false)
   MakeLiteralText(Value="LOCTEXT(\"Quadrant Top Left\\r\\n\\r\\nRed = Forest\\r\\nGreen = BrightGrass\\r\\nBlue = Unused\\r\\n\\r\\nQuadrant Top Right\\r\\n\\r\\nRed = Springfield Gorge\\r\\nGreen = Unused\\r\\nBlue = Unused\\r\\n\\r\\n\")")
   Get Actor Bounds(self=@n12, bOnlyCollidingComponents=false)
   Branch()
   Get Lanscape Actor()
  ]}
]
variables [
 BiomeTexture {type=Texture, default=&/T_BiomeMask_Template}
 Enabled {type=boolean}
 LandscapeComponentSize {type=number, default=16256}
 Lanscape Actor {type=Landscape, default=null}
 Meshes {type="array<StaticMeshComponent>", default [null,null,null,null]}
 Number of Landscape Tiles {type=number, default=16}
 Paint Region {type=BiomeMaskPaintingRegion, default=Quadrant Top Left}
 Quadrant Bottom Left {type="array<string>", default [Unused,Unused,Unused]}
 Quadrant Bottom Right {type="array<string>", default [Unused,Unused,Unused]}
 Quadrant Top Left {type="array<string>", default [Forest,BrightGrass,Unused]}
 Quadrant Top Right {type="array<string>", default [Gorge,Unused,Unused]}
 Usage {type=string, default=""}
]
